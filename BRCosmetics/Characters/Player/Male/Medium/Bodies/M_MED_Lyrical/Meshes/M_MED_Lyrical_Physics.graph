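# Physics asset for the M_MED_Lyrical male player body, simulating the hanging parts of its jacket.
# Nine jacket chains (front, side, back) each run from a small master capsule, linked to spine_03 with free rotation, through three capsule segments.
# Segment joints have per-chain swing limits and very tight twist to restrict flapping and rolling.
# Capsules on the pelvis, spine and thighs supply the body volume around the jacket.
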PhysicsAsset /BRCosmetics/Characters/Player/Male/Medium/Bodies/M_MED_Lyrical/Meshes/M_MED_Lyrical_Physics
bodies {
 dyn_jacket_bk_01 [{shapeType=Capsule, length=5, radius=3, shapeName=dyn_jacket_fr_r_01_capsule, center=<4 0 1.6862>, rotation=<rot 90 0 0>}]
 dyn_jacket_bk_02 [{shapeType=Capsule, length=5, radius=3, shapeName=dyn_jacket_fr_r_02_capsule, center=<4 0 1.6862>, rotation=<rot 90 0 0>}]
 dyn_jacket_bk_03 [{shapeType=Capsule, length=2, radius=3, shapeName=dyn_jacket_fr_r_03_capsule, center=<4.9592 0 0.5301>, rotation=<rot 0 0 90>}]
 dyn_jacket_bk_mstr [{shapeType=Capsule, length=0.505, radius=0.505, shapeName=dyn_jacket_bk_mstr_capsule, rotation=<rot 0 -90 0>}]
 dyn_jacket_bks_01_l_01 [{shapeType=Capsule, length=5, radius=3, shapeName=dyn_jacket_fr_r_01_capsule, center=<4 0 1.6862>, rotation=<rot 90 0 0>}]
 dyn_jacket_bks_01_l_02 [{shapeType=Capsule, length=5, radius=3, shapeName=dyn_jacket_fr_r_02_capsule, center=<4 0 1.6862>, rotation=<rot 90 0 0>}]
 dyn_jacket_bks_01_l_03 [{
   shapeType=Capsule
   length=10
   radius=3.25
   shapeName=dyn_jacket_fr_r_03_capsule
   center=<4.9592 0.0021 0.6229>
   rotation=<rot -0.0001 0.0001 91.2772>
  }]
 dyn_jacket_bks_01_l_mstr [{shapeType=Capsule, length=0.505, radius=0.505, shapeName=dyn_jacket_bks_01_l_mstr_capsule, rotation=<rot 0 -90 0>}]
 dyn_jacket_bks_01_r_01 [{shapeType=Capsule, length=5, radius=3, shapeName=dyn_jacket_fr_r_01_capsule, center=<4 0 1.6862>, rotation=<rot 90 0 0>}]
 dyn_jacket_bks_01_r_02 [{shapeType=Capsule, length=5, radius=3, shapeName=dyn_jacket_fr_r_02_capsule, center=<4 0 1.6862>, rotation=<rot 90 0 0>}]
 dyn_jacket_bks_01_r_03 [{
   shapeType=Capsule
   length=10
   radius=3.25
   shapeName=dyn_jacket_fr_r_03_capsule
   center=<4.9592 0.0021 0.6229>
   rotation=<rot -0.0001 0.0001 91.2772>
  }]
 dyn_jacket_bks_01_r_mstr [{shapeType=Capsule, length=0.505, radius=0.505, shapeName=dyn_jacket_bks_01_r_mstr_capsule, rotation=<rot 0 -90 0>}]
 dyn_jacket_bks_02_l_01 [{shapeType=Capsule, length=5, radius=3, shapeName=dyn_jacket_fr_r_01_capsule, center=<4 0 2.0579>, rotation=<rot 90 0 0>}]
 dyn_jacket_bks_02_l_02 [{shapeType=Capsule, length=5, radius=3, shapeName=dyn_jacket_fr_r_02_capsule, center=<4 0 1.6862>, rotation=<rot 90 0 0>}]
 dyn_jacket_bks_02_l_03 [{shapeType=Capsule, length=8.25, radius=3.25, shapeName=dyn_jacket_fr_r_03_capsule, center=<2.2516 0 0.718>, rotation=<rot 0 0 90>}]
 dyn_jacket_bks_02_l_mstr [{shapeType=Capsule, length=0.505, radius=0.505, shapeName=dyn_jacket_bks_02_l_mstr_capsule, rotation=<rot 0 -90 0>}]
 dyn_jacket_bks_02_r_01 [{shapeType=Capsule, length=5, radius=3, shapeName=dyn_jacket_fr_r_01_capsule, center=<4 0 2.0579>, rotation=<rot 90 0 0>}]
 dyn_jacket_bks_02_r_02 [{shapeType=Capsule, length=5, radius=3, shapeName=dyn_jacket_fr_r_02_capsule, center=<4 0 1.6862>, rotation=<rot 90 0 0>}]
 dyn_jacket_bks_02_r_03 [{shapeType=Capsule, length=7.75, radius=3.75, shapeName=dyn_jacket_fr_r_03_capsule, center=<2.2516 0 0.718>, rotation=<rot 0 0 90>}]
 dyn_jacket_bks_02_r_mstr [{shapeType=Capsule, length=0.505, radius=0.505, shapeName=dyn_jacket_bks_02_r_mstr_capsule, rotation=<rot 0 -90 0>}]
 dyn_jacket_fr_l_01 [{
   shapeType=Capsule
   length=8.75
   radius=2.75
   shapeName=dyn_jacket_fr_r_01_capsule
   center=<4.9 0.2981 0.6264>
   rotation=<rot 88.0014 0.0001 0>
  }]
 dyn_jacket_fr_l_02 [{
   shapeType=Capsule
   length=6.25
   radius=3
   shapeName=dyn_jacket_fr_r_02_capsule
   center=<4.0404 -1.1406 0.741>
   rotation=<rot 0.0001 0.0001 85.1845>
  }]
 dyn_jacket_fr_l_03 [
  {
   shapeType=Capsule
   length=6.25
   radius=3.5
   shapeName=dyn_jacket_fr_r_03_capsule
   center=<3.9441 -1.1049 0.8589>
   rotation=<rot 0.0001 0 86.9256>
  }
  {shapeType=Capsule, length=2, radius=2, shapeName=dyn_jacket_fr_l_03_capsule, center=<3.9582 -9.4604 -2.7016>, rotation=<rot 0 0 48.8558>}
 ]
 dyn_jacket_fr_l_mstr [{shapeType=Capsule, length=1, radius=1}]
 dyn_jacket_fr_r_01 [{shapeType=Capsule, length=5, radius=3, shapeName=dyn_jacket_fr_r_01_capsule, center=<4 0 1.6862>, rotation=<rot 90 0 0>}]
 dyn_jacket_fr_r_02 [{
   shapeType=Capsule
   length=6.25
   radius=3
   shapeName=dyn_jacket_fr_r_02_capsule
   center=<4.0404 -1.1741 1.1382>
   rotation=<rot 0.0001 0.0001 85.1845>
  }]
 dyn_jacket_fr_r_03 [{shapeType=Capsule, length=7.25, radius=3.5, shapeName=dyn_jacket_fr_r_03_capsule, center=<3.9441 1.9 0.025>, rotation=<rot 0.0001 0 100>}]
 dyn_jacket_fr_r_mstr [{shapeType=Capsule, length=0.505, radius=0.505, shapeName=dyn_jacket_fr_r_mstr_capsule, rotation=<rot 0 -90 0>}]
 dyn_jacket_s_l_01 [{shapeType=Capsule, length=5, radius=3, shapeName=dyn_jacket_fr_r_01_capsule, center=<4 0 1.6862>, rotation=<rot 90 0 0>}]
 dyn_jacket_s_l_02 [{shapeType=Capsule, length=5, radius=3, shapeName=dyn_jacket_fr_r_02_capsule, center=<4 0 2.5962>, rotation=<rot 0 0 90>}]
 dyn_jacket_s_l_03 [{shapeType=Capsule, length=7.75, radius=2, shapeName=dyn_jacket_fr_r_03_capsule, center=<2.6241 0.141 0.1583>, rotation=<rot 0 0 81.8357>}]
 dyn_jacket_s_l_mstr [{shapeType=Capsule, length=0.505, radius=0.505, shapeName=dyn_jacket_s_l_mstr_capsule, rotation=<rot 0 -90 0>}]
 dyn_jacket_s_r_01 [{shapeType=Capsule, length=5, radius=3, shapeName=dyn_jacket_fr_r_01_capsule, center=<4 0 1.6862>, rotation=<rot 90 0 0>}]
 dyn_jacket_s_r_02 [{shapeType=Capsule, length=5, radius=3, shapeName=dyn_jacket_fr_r_02_capsule, center=<4 0 1.6862>, rotation=<rot 90 0 0>}]
 dyn_jacket_s_r_03 [{shapeType=Capsule, length=10.75, radius=2, shapeName=dyn_jacket_fr_r_03_capsule, center=<2.6241 0.141 0.1583>, rotation=<rot 0 0 -81.8357>}]
 dyn_jacket_s_r_mstr [{shapeType=Capsule, length=0.505, radius=0.505, shapeName=dyn_jacket_s_r_mstr_capsule, rotation=<rot 0 -90 0>}]
 pelvis [{shapeType=Capsule, length=8.5938, radius=10.2812, center=<-1.4216 -1.0559 -0.0001>}]
 spine_01 [{shapeType=Capsule, length=8.7188, radius=9.6875, center=<-0.8589 -0.1368 0.0001>}]
 spine_02 [{shapeType=Capsule, length=7.125, radius=10.8438, center=<-0.1537 0.0199 -0.0002>}]
 spine_03 [{shapeType=Capsule, length=7.7188, radius=10.25, center=<1.2451 0.0138 -0.0001>}]
 thigh_l [{shapeType=Capsule, length=27.9688, radius=10.5938, center=<-15.1488 -0.9527 2.4766>, rotation=<rot 88.5265 -9.6052 -10.4338>}]
 thigh_r [{shapeType=Capsule, length=27.2188, radius=10, center=<17.057 0.2342 -2.0844>, rotation=<rot 88.5265 -9.6056 -12.5396>}]
}
constraints [
 {bone1Name=spine_03, bone2Name=spine_02, swing1LimitDegrees=45, swing2LimitDegrees=45, twistLimitDegrees=45}
 {
  bone1Name=spine_01
  bone2Name=pelvis
  swing1LimitDegrees=45
  swing1Motion=Free
  swing2LimitDegrees=45
  swing2Motion=Free
  twistLimitDegrees=45
  twistMotion=Free
 }
 {
  bone1Name=spine_02
  bone2Name=spine_01
  swing1LimitDegrees=45
  swing1Motion=Free
  swing2LimitDegrees=45
  swing2Motion=Free
  twistLimitDegrees=45
  twistMotion=Free
 }
 {
  bone1Name=thigh_l
  bone2Name=pelvis
  swing1LimitDegrees=45
  swing1Motion=Free
  swing2LimitDegrees=45
  swing2Motion=Free
  twistLimitDegrees=45
  twistMotion=Free
 }
 {
  bone1Name=thigh_r
  bone2Name=pelvis
  swing1LimitDegrees=45
  swing1Motion=Free
  swing2LimitDegrees=45
  swing2Motion=Free
  twistLimitDegrees=45
  twistMotion=Free
 }
 {
  bone1Name=dyn_jacket_s_l_mstr
  bone2Name=spine_03
  swing1LimitDegrees=45
  swing1Motion=Free
  swing2LimitDegrees=45
  swing2Motion=Free
  twistLimitDegrees=45
  twistMotion=Free
 }
 {
  bone1Name=dyn_jacket_fr_l_mstr
  bone2Name=spine_03
  swing1LimitDegrees=45
  swing1Motion=Free
  swing2LimitDegrees=45
  swing2Motion=Free
  twistLimitDegrees=45
  twistMotion=Free
 }
 {
  bone1Name=dyn_jacket_bks_02_l_mstr
  bone2Name=spine_03
  swing1LimitDegrees=45
  swing1Motion=Free
  swing2LimitDegrees=45
  swing2Motion=Free
  twistLimitDegrees=45
  twistMotion=Free
 }
 {
  bone1Name=dyn_jacket_bk_mstr
  bone2Name=spine_03
  swing1LimitDegrees=45
  swing1Motion=Free
  swing2LimitDegrees=45
  swing2Motion=Free
  twistLimitDegrees=45
  twistMotion=Free
 }
 {
  bone1Name=dyn_jacket_bks_01_l_mstr
  bone2Name=spine_03
  swing1LimitDegrees=45
  swing1Motion=Free
  swing2LimitDegrees=45
  swing2Motion=Free
  twistLimitDegrees=45
  twistMotion=Free
 }
 {
  bone1Name=dyn_jacket_fr_r_mstr
  bone2Name=spine_03
  swing1LimitDegrees=45
  swing1Motion=Free
  swing2LimitDegrees=45
  swing2Motion=Free
  twistLimitDegrees=45
  twistMotion=Free
 }
 {
  bone1Name=dyn_jacket_bks_01_r_mstr
  bone2Name=spine_03
  swing1LimitDegrees=45
  swing1Motion=Free
  swing2LimitDegrees=45
  swing2Motion=Free
  twistLimitDegrees=45
  twistMotion=Free
 }
 {
  bone1Name=dyn_jacket_bks_02_r_mstr
  bone2Name=spine_03
  swing1LimitDegrees=45
  swing1Motion=Free
  swing2LimitDegrees=45
  swing2Motion=Free
  twistLimitDegrees=45
  twistMotion=Free
 }
 {
  bone1Name=dyn_jacket_s_r_mstr
  bone2Name=spine_03
  swing1LimitDegrees=45
  swing1Motion=Free
  swing2LimitDegrees=45
  swing2Motion=Free
  twistLimitDegrees=45
  twistMotion=Free
 }
 {bone1Name=dyn_jacket_fr_r_01, bone2Name=dyn_jacket_fr_r_mstr, swing1LimitDegrees=10, swing2LimitDegrees=45, twistLimitDegrees=2}
 {bone1Name=dyn_jacket_fr_r_02, bone2Name=dyn_jacket_fr_r_01, swing1LimitDegrees=15, swing2LimitDegrees=45, twistLimitDegrees=2}
 {bone1Name=dyn_jacket_fr_r_03, bone2Name=dyn_jacket_fr_r_02, swing1LimitDegrees=15, swing2LimitDegrees=60, twistLimitDegrees=2}
 {bone1Name=dyn_jacket_s_r_01, bone2Name=dyn_jacket_s_r_mstr, swing1LimitDegrees=10, swing2LimitDegrees=45, twistLimitDegrees=2}
 {bone1Name=dyn_jacket_s_r_02, bone2Name=dyn_jacket_s_r_01, swing1LimitDegrees=20, swing2LimitDegrees=45, twistLimitDegrees=2}
 {bone1Name=dyn_jacket_s_r_03, bone2Name=dyn_jacket_s_r_02, swing1LimitDegrees=40, swing2LimitDegrees=60, twistLimitDegrees=2}
 {bone1Name=dyn_jacket_bks_02_r_01, bone2Name=dyn_jacket_bks_02_r_mstr, swing1LimitDegrees=20, swing2LimitDegrees=45, twistLimitDegrees=2}
 {bone1Name=dyn_jacket_bks_02_r_02, bone2Name=dyn_jacket_bks_02_r_01, swing1LimitDegrees=10, swing2LimitDegrees=45, twistLimitDegrees=2}
 {bone1Name=dyn_jacket_bks_02_r_03, bone2Name=dyn_jacket_bks_02_r_02, swing1LimitDegrees=10, swing2LimitDegrees=60, twistLimitDegrees=2}
 {bone1Name=dyn_jacket_bks_01_r_01, bone2Name=dyn_jacket_bks_01_r_mstr, swing1LimitDegrees=10, swing2LimitDegrees=45, twistLimitDegrees=2}
 {bone1Name=dyn_jacket_bks_01_r_02, bone2Name=dyn_jacket_bks_01_r_01, swing1LimitDegrees=10, swing2LimitDegrees=45, twistLimitDegrees=2}
 {bone1Name=dyn_jacket_bks_01_r_03, bone2Name=dyn_jacket_bks_01_r_02, swing1LimitDegrees=10, swing2LimitDegrees=60, twistLimitDegrees=2}
 {bone1Name=dyn_jacket_bk_03, bone2Name=dyn_jacket_bk_02, swing1LimitDegrees=10, swing2LimitDegrees=60, twistLimitDegrees=2}
 {bone1Name=dyn_jacket_bk_02, bone2Name=dyn_jacket_bk_01, swing1LimitDegrees=10, swing2LimitDegrees=45, twistLimitDegrees=2}
 {bone1Name=dyn_jacket_bk_01, bone2Name=dyn_jacket_bk_mstr, swing1LimitDegrees=10, swing2LimitDegrees=30, twistLimitDegrees=2}
 {bone1Name=dyn_jacket_bks_01_l_01, bone2Name=dyn_jacket_bks_01_l_mstr, swing1LimitDegrees=10, swing2LimitDegrees=30, twistLimitDegrees=2}
 {bone1Name=dyn_jacket_bks_01_l_02, bone2Name=dyn_jacket_bks_01_l_01, swing1LimitDegrees=10, swing2LimitDegrees=45, twistLimitDegrees=2}
 {bone1Name=dyn_jacket_bks_01_l_03, bone2Name=dyn_jacket_bks_01_l_02, swing1LimitDegrees=10, swing2LimitDegrees=60, twistLimitDegrees=2}
 {bone1Name=dyn_jacket_bks_02_l_01, bone2Name=dyn_jacket_bks_02_l_mstr, swing1LimitDegrees=10, swing2LimitDegrees=25, twistLimitDegrees=2}
 {bone1Name=dyn_jacket_bks_02_l_02, bone2Name=dyn_jacket_bks_02_l_01, swing1LimitDegrees=10, swing2LimitDegrees=45, twistLimitDegrees=2}
 {bone1Name=dyn_jacket_bks_02_l_03, bone2Name=dyn_jacket_bks_02_l_02, swing1LimitDegrees=10, swing2LimitDegrees=60, twistLimitDegrees=2}
 {bone1Name=dyn_jacket_fr_l_01, bone2Name=dyn_jacket_fr_l_mstr, swing1LimitDegrees=10, swing2LimitDegrees=40, twistLimitDegrees=2}
 {bone1Name=dyn_jacket_fr_l_02, bone2Name=dyn_jacket_fr_l_01, swing1LimitDegrees=20, swing2LimitDegrees=60, twistLimitDegrees=3}
 {bone1Name=dyn_jacket_fr_l_03, bone2Name=dyn_jacket_fr_l_02, swing1LimitDegrees=15, swing2LimitDegrees=60, twistLimitDegrees=2}
 {bone1Name=dyn_jacket_s_l_01, bone2Name=dyn_jacket_s_l_mstr, swing1LimitDegrees=10, swing2LimitDegrees=25, twistLimitDegrees=2}
 {bone1Name=dyn_jacket_s_l_02, bone2Name=dyn_jacket_s_l_01, swing1LimitDegrees=20, swing2LimitDegrees=40, twistLimitDegrees=2}
 {bone1Name=dyn_jacket_s_l_03, bone2Name=dyn_jacket_s_l_02, swing1LimitDegrees=40, swing2LimitDegrees=60, twistLimitDegrees=2}
]
