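# Animation Blueprint for the M_MED_Lyrical player body cosmetic, animating its jacket and collar over the pose copied from the parent mesh.
# A Control Rig, active when rigid body is disabled, feeds RigidBody physics with state-driven gravity, then additive bone correctives.
# Event graph maps thigh, neck and head deltas from ref pose into corrective alphas, some gated by crouch or jump/fall states.
# Crouch transitions reset dynamics.
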AnimBlueprint /BRCosmetics/Characters/Player/Male/Medium/Bodies/M_MED_Lyrical/Meshes/M_MED_Lyrical_AnimBP
target_skeleton=&/M_MED_Lyrical_Skeleton
graphs [
 AnimGraph {nodes [
   Control Rig!(Source, Alpha, IsCrouchingCtrl:bool=Get CrouchingAnim() -> (CrouchingAnim:bool)) -> (Pose) {
    alphaBoolBlend {blendOption=Linear}
    controlRigAssetReference {blueprintRigClass=&/M_MED_Lyrical_CtrlRig.M_MED_Lyrical_CtrlRig_C}
   }
   "Transform (Modify) Bone - Bone: dyn_collar_back"(Alpha=0, Translation="-0.133,0.572,2.677", Rotation="-20.000,-0.000,50", Scale="1,1,1") {
    rotationMode=BMM_Additive
    rotationSpace=BCS_BoneSpace
    scaleMode=BMM_Additive
    scaleSpace=BCS_BoneSpace
    translationMode=BMM_Additive
    translationSpace=BCS_BoneSpace
    alphaBoolBlend {blendOption=Linear}
    alphaScaleBiasClamp {interpSpeedDecreasing=7, interpSpeedIncreasing=7}
    boneToModify {boneName=dyn_collar_back}
    rotation {pitch=-20, roll=50}
    translation=<-0.133 0.572 2.677>
   }
   n1 = Get Extremewalkup() -> (Extremewalkup:double)
   n2 = Get thigh_rot_r()
   n3 = Control Rig!(Source=Copy Pose From Mesh() {bUseAttachedParent=true}, Alpha=Get ControlRigAlpha() -> (ControlRigAlpha:double)) -> (Pose) {
    alphaBoolBlend {blendOption=Linear}
    controlRigAssetReference {blueprintRigClass=&M_MED_Lyrical_CtrlRig_C}
   }
   n4 = RigidBody!(ComponentPose=Local To Component(LocalPose=@n3), Alpha, OverrideWorldGravity:Vector=Get GravityOverride(), ExternalForce) -> (Pose) {
    bClampLinearTranslationLimitToRefPose=true
    bOverrideWorldGravity=true
    overridePhysicsAsset=&/M_MED_Lyrical_Physics
    simulationSpace=BaseBoneSpace
    alphaBoolBlend {blendOption=Linear}
    baseBoneRef {boneName=pelvis}
    componentAppliedLinearAccClamp=<2000 2000 2000>
    componentLinearAccScale=<0.05 0.05 0.05>
    componentLinearVelScale=<0.05 0.05 0.05>
    overrideWorldGravity {z=-980}
    simSpaceSettings {
     maxAngularAcceleration=300
     maxAngularVelocity=3
     maxLinearAcceleration=500
     maxLinearVelocity=300
     velocityScaleZ=0
     worldAlpha=1
     externalLinearDragV=<1 1>
    }
   }
   n5 = Get thigh_rot_l()
   n6 = "Transform (Modify) Bone - Bone: dyn_jacket_fr_l_mstr"(ComponentPose=@n4, Alpha=@n5, Translation="0,-1,1", Rotation="20,0,0", Scale="1,1,1") {
    rotationMode=BMM_Additive
    scaleMode=BMM_Additive
    scaleSpace=BCS_BoneSpace
    translationMode=BMM_Additive
    translationSpace=BCS_BoneSpace
    alphaBoolBlend {blendOption=Linear}
    alphaScaleBiasClamp {bInterpResult=true, interpSpeedDecreasing=9, interpSpeedIncreasing=30}
    boneToModify {boneName=dyn_jacket_fr_l_mstr}
    rotation {pitch=20}
    translation {y=-1, z=1}
   }
   n7 = "Transform (Modify) Bone - Bone: dyn_jacket_s_l_mstr"(ComponentPose=@n6, Alpha=@n5, Translation="0,0,0", Rotation="25,0,0", Scale="1,1,1") {
    rotationMode=BMM_Additive
    scaleMode=BMM_Additive
    scaleSpace=BCS_BoneSpace
    translationMode=BMM_Additive
    translationSpace=BCS_BoneSpace
    alphaBoolBlend {blendOption=Linear}
    alphaScaleBiasClamp {bInterpResult=true, interpSpeedDecreasing=7, interpSpeedIncreasing=7}
    boneToModify {boneName=dyn_jacket_s_l_mstr}
    rotation {pitch=25}
   }
   n8 = "Transform (Modify) Bone - Bone: dyn_jacket_fr_r_mstr"(ComponentPose=@n7, Alpha=@n2, Translation="-2,0,0", Rotation="-53.573,134.686,161.849", Scale="1,1,1") {
    rotationMode=BMM_Additive
    scaleMode=BMM_Additive
    scaleSpace=BCS_BoneSpace
    translationMode=BMM_Additive
    translationSpace=BCS_BoneSpace
    alphaBoolBlend {blendOption=Linear}
    boneToModify {boneName=dyn_jacket_fr_r_mstr}
    rotation=<rot -53.573 134.686 161.849>
    translation {x=-2}
   }
   n9 = "Transform (Modify) Bone - Bone: dyn_jacket_s_r_mstr"(ComponentPose=@n8, Alpha=@n2, Translation="0,0,0", Rotation="-58.231,-20.966,-9.639", Scale="1,1,1") {
    rotationMode=BMM_Additive
    scaleMode=BMM_Additive
    scaleSpace=BCS_BoneSpace
    translationMode=BMM_Additive
    translationSpace=BCS_BoneSpace
    alphaBoolBlend {blendOption=Linear}
    alphaScaleBiasClamp {bInterpResult=true, interpSpeedDecreasing=7, interpSpeedIncreasing=7}
    boneToModify {boneName=dyn_jacket_s_r_mstr}
    rotation=<rot -58.231 -20.966 -9.639>
   }
   n10 = "Transform (Modify) Bone - Bone: dyn_jacket_fr_r_mstr"(
    ComponentPose=@n9
    Alpha=Get Thigh R Crouch() -> (Thigh R Crouch:double)
    Translation="0,0,0"
    Rotation="-19.105,8.396,-47.135"
    Scale="1,1,1"
   ) {
    rotationMode=BMM_Additive
    alphaBoolBlend {blendOption=Linear}
    alphaScaleBiasClamp {bInterpResult=true}
    boneToModify {boneName=dyn_jacket_fr_r_mstr}
    rotation=<rot -19.105 8.396 -47.135>
   }
   n11 = "Transform (Modify) Bone - Bone: dyn_collar_base"(
    ComponentPose=@n10
    Alpha=Get Head Tilt Alpha() -> (Head Tilt Alpha:double)
    Translation="1,-0.000,-1"
    Rotation="-55,0.000,0"
    Scale="1,1,1"
   ) {
    rotationMode=BMM_Additive
    rotationSpace=BCS_BoneSpace
    scaleMode=BMM_Additive
    scaleSpace=BCS_BoneSpace
    translationMode=BMM_Additive
    translationSpace=BCS_BoneSpace
    alphaBoolBlend {blendOption=Linear}
    alphaScaleBiasClamp {interpSpeedDecreasing=7, interpSpeedIncreasing=7}
    boneToModify {boneName=dyn_collar_base}
    rotation {pitch=-55}
    translation {x=1, z=-1}
   }
   n12 = "Transform (Modify) Bone - Bone: dyn_collar_right"(
    ComponentPose=@n11
    Alpha=Get Neck R Alpha() -> (Neck R Alpha:double)
    Translation="0.197,0.000,-3"
    Rotation="-40,0.000,-0.000"
    Scale="1,1,1"
   ) {
    rotationMode=BMM_Additive
    rotationSpace=BCS_BoneSpace
    scaleMode=BMM_Additive
    scaleSpace=BCS_BoneSpace
    translationMode=BMM_Additive
    translationSpace=BCS_BoneSpace
    alphaBoolBlend {blendOption=Linear}
    alphaScaleBiasClamp {interpSpeedDecreasing=7, interpSpeedIncreasing=7}
    boneToModify {boneName=dyn_collar_right}
    rotation {pitch=-40}
    translation {x=0.197, z=-3}
   }
   n13 = "Transform (Modify) Bone - Bone: dyn_collar_left"(
    ComponentPose=@n12
    Alpha=Get Neck L Alpha() -> (Neck L Alpha:double)
    Translation="0.073,-1.498,-0.247"
    Rotation="-73.274,0.000,-0.000"
    Scale="1,1,1"
   ) {
    rotationMode=BMM_Additive
    rotationSpace=BCS_BoneSpace
    scaleMode=BMM_Additive
    scaleSpace=BCS_BoneSpace
    translationMode=BMM_Additive
    translationSpace=BCS_BoneSpace
    alphaBoolBlend {blendOption=Linear}
    alphaScaleBiasClamp {interpSpeedDecreasing=7, interpSpeedIncreasing=7}
    boneToModify {boneName=dyn_collar_left}
    rotation=<rot -73.274 0 0>
    translation=<0.073 -1.498 -0.247>
   }
   n14 = "Transform (Modify) Bone - Bone: dyn_collar_base"(ComponentPose=@n13, Alpha=1, Translation="0,0,0", Rotation="-12,-0.000,0.000", Scale="1,1,1") {
    rotationMode=BMM_Additive
    rotationSpace=BCS_BoneSpace
    scaleMode=BMM_Additive
    scaleSpace=BCS_BoneSpace
    translationMode=BMM_Additive
    translationSpace=BCS_BoneSpace
    alphaBoolBlend {blendOption=Linear}
    alphaScaleBiasClamp {interpSpeedDecreasing=7, interpSpeedIncreasing=7}
    boneToModify {boneName=dyn_collar_base}
    rotation {pitch=-12}
   }
   n15 = "Transform (Modify) Bone - Bone: dyn_collar_base"(ComponentPose=@n14, Alpha=Get Neck Rcrouch() -> (Neck Rcrouch:double), Translation="2,0.000,0", Rotation="-13,0.000,-0.000", Scale="1,1,1") {
    rotationMode=BMM_Additive
    rotationSpace=BCS_BoneSpace
    scaleMode=BMM_Additive
    scaleSpace=BCS_BoneSpace
    translationMode=BMM_Additive
    translationSpace=BCS_BoneSpace
    alphaBoolBlend {blendOption=Linear}
    alphaScaleBiasClamp {interpSpeedDecreasing=7, interpSpeedIncreasing=7}
    boneToModify {boneName=dyn_collar_base}
    rotation {pitch=-13}
    translation {x=2}
   }
   n16 = "Transform (Modify) Bone - Bone: dyn_collar_back_r"(
    ComponentPose=@n15
    Alpha=Get Neck twist R() -> (Neck twist R:double)
    Translation="0.679,0.000,-0.390"
    Rotation="-28.147,22.076,-10.830"
    Scale="1,1,1"
   ) {
    rotationMode=BMM_Additive
    rotationSpace=BCS_BoneSpace
    scaleMode=BMM_Additive
    scaleSpace=BCS_BoneSpace
    translationMode=BMM_Additive
    translationSpace=BCS_BoneSpace
    alphaBoolBlend {blendOption=Linear}
    alphaScaleBiasClamp {interpSpeedDecreasing=7, interpSpeedIncreasing=7}
    boneToModify {boneName=dyn_collar_back_r}
    rotation=<rot -28.1471 22.076 -10.83>
    translation=<0.6789 0 -0.3899>
   }
   n17 = "Transform (Modify) Bone - Bone: dyn_collar_back_r"(ComponentPose=@n16, Alpha=@n1, Translation="4.041,0.000,-1.950", Rotation="-50,0.000,-0.000", Scale="1,1,1") {
    rotationMode=BMM_Additive
    rotationSpace=BCS_BoneSpace
    scaleMode=BMM_Additive
    scaleSpace=BCS_BoneSpace
    translationMode=BMM_Additive
    translationSpace=BCS_BoneSpace
    alphaBoolBlend {blendOption=Linear}
    alphaScaleBiasClamp {interpSpeedDecreasing=7, interpSpeedIncreasing=7}
    boneToModify {boneName=dyn_collar_back_r}
    rotation {pitch=-50}
    translation {x=4.041, z=-1.95}
   }
   n18 = "Transform (Modify) Bone - Bone: dyn_collar_back"(ComponentPose=@n17, Alpha=@n1, Translation="-0.133,0.572,2.677", Rotation="-20.000,-0.000,66.936", Scale="1,1,1") {
    rotationMode=BMM_Additive
    rotationSpace=BCS_BoneSpace
    scaleMode=BMM_Additive
    scaleSpace=BCS_BoneSpace
    translationMode=BMM_Additive
    translationSpace=BCS_BoneSpace
    alphaBoolBlend {blendOption=Linear}
    alphaScaleBiasClamp {interpSpeedDecreasing=7, interpSpeedIncreasing=7}
    boneToModify {boneName=dyn_collar_back}
    rotation=<rot -20 0 66.936>
    translation=<-0.133 0.572 2.677>
   }
   n19 = "Transform (Modify) Bone - Bone: dyn_collar_back_r"(
    ComponentPose=@n18
    Alpha=Get NeckXCrouch() -> (NeckXCrouch:double)
    Translation="2.692,-0.000,0.199"
    Rotation="-20,-0.000,0.000"
    Scale="1,1,1"
   ) {
    rotationMode=BMM_Additive
    rotationSpace=BCS_BoneSpace
    scaleMode=BMM_Additive
    scaleSpace=BCS_BoneSpace
    translationMode=BMM_Additive
    translationSpace=BCS_BoneSpace
    alphaBoolBlend {blendOption=Linear}
    alphaScaleBiasClamp {interpSpeedDecreasing=7, interpSpeedIncreasing=7}
    boneToModify {boneName=dyn_collar_back_r}
    rotation {pitch=-20}
    translation {x=2.692, z=0.199}
   }
   n20 = "Transform (Modify) Bone - Bone: dyn_collar_back"(ComponentPose=@n19, Alpha=Get NeckYDown() -> (NeckYDown:double), Translation="0.000,-0.000,1.3", Rotation="0,0,0", Scale="1,1,1") {
    rotationMode=BMM_Additive
    rotationSpace=BCS_BoneSpace
    scaleMode=BMM_Additive
    scaleSpace=BCS_BoneSpace
    translationMode=BMM_Additive
    translationSpace=BCS_BoneSpace
    alphaBoolBlend {blendOption=Linear}
    alphaScaleBiasClamp {interpSpeedDecreasing=7, interpSpeedIncreasing=7}
    boneToModify {boneName=dyn_collar_back}
    translation {z=1.3}
   }
   Output Pose(Result=Component To Local(ComponentPose=@n20)) {layerGroup=None}
   Get Neck Alpha() -> (Neck Alpha:double)
  ]}
 EventGraph {nodes [
   Event BlueprintUpdateAnimation()
   > n1 = Sequence!(execute) -> (then_0, then_1, then_2:Exec, then_3:Exec, then_4:Exec)
   > n2 = Gravity Override State Function DEPRECATED(CharacterPartAnimInstance=Self-Reference(), GameTimeInSeconds=GetGameTimeInSeconds(), GravityOverrideParams=Get Gravity Override Params())
   > Set GravityOverride(GravityOverride=@n2)
   n3 = Branch(execute=@n1.then_1, Condition=OR Boolean!(A=false, B=Get bIsCrouchSprinting(), C:bool=Get bIsCrouching()) -> (ReturnValue))
   n4 = BreakTransform(InTransform=GetDeltaTransformFromRefPose(self=GetOwnerSkeletalMeshComponent(), BoneName=thigh_l, BaseName=pelvis))
   > Set thigh_rot_l(thigh_rot_l=MapRangeClamped(Value=BreakRotator(InRot=@n4.Rotation).Yaw, InRangeA=-100, InRangeB=-20, OutRangeA=1, OutRangeB=0))
   Set thigh_rot_l(execute=@n3.else, thigh_rot_l=0)
   Set ControlRigAlpha(execute=@n1.then_4, ControlRigAlpha:double=SelectFloat(A=0, B=1, bPickA=Get bRigidBodyEnabled())) -> (Output_Get:double)
   n5 = BreakTransform(InTransform=GetDeltaTransformFromRefPose(self=GetOwnerSkeletalMeshComponent(), BoneName=neck_01, BaseName=spine_05))
   Set Head Tilt Alpha(execute=@n1.then_2, Head Tilt Alpha:double=MapRangeClamped(Value=BreakRotator(InRot=@n5.Rotation).Yaw, InRangeA=-15, InRangeB=-50)) -> (Output_Get:double)
   n6 = BreakTransform(InTransform=GetDeltaTransformFromRefPose(self=GetOwnerSkeletalMeshComponent(), BoneName=neck_01, BaseName=spine_05))
   > Set Neck R Alpha(Neck R Alpha:double=MapRangeClamped(Value=BreakRotator(InRot=@n6.Rotation).Pitch, InRangeA=10, InRangeB=25)) -> (Output_Get:double)
   n7 = BreakTransform(InTransform=GetDeltaTransformFromRefPose(self=GetOwnerSkeletalMeshComponent(), BoneName=neck_01, BaseName=spine_05))
   > Set Neck L Alpha(Neck L Alpha:double=MapRangeClamped(Value=BreakRotator(InRot=@n7.Rotation).Pitch, InRangeA=-2, InRangeB=-10)) -> (Output_Get:double)
   n8 = BreakTransform(InTransform=GetDeltaTransformFromRefPose(self=GetOwnerSkeletalMeshComponent(), BoneName=neck_01, BaseName=spine_05))
   > Set Neck twist R(Neck twist R:double=MapRangeClamped(Value=BreakRotator(InRot=@n8.Rotation), InRangeA=-2, InRangeB=-10)) -> (Output_Get:double)
   > n9 = Branch(Condition=OR Boolean!(A=Get bIsCrouchSprinting(), B=Get bIsCrouching()) -> (ReturnValue))
   n10 = BreakTransform(InTransform=GetDeltaTransformFromRefPose(self=GetOwnerSkeletalMeshComponent(), BoneName=neck_01, BaseName=spine_05))
   > Set Neck Alpha(Neck Alpha:double=MapRangeClamped(Value=BreakRotator(InRot=@n10.Rotation).Yaw, InRangeA=-45, InRangeB=-55)) -> (Output_Get:double)
   n11 = BreakTransform(InTransform=GetDeltaTransformFromRefPose(self=GetOwnerSkeletalMeshComponent(), BoneName=head, BaseName=spine_05))
   > Set Neck Rcrouch(Neck Rcrouch:double=MapRangeClamped(Value=BreakRotator(InRot=@n11.Rotation).Yaw, InRangeA=-50, InRangeB=-90)) -> (Output_Get:double)
   n12 = BreakTransform(InTransform=GetDeltaTransformFromRefPose(self=GetOwnerSkeletalMeshComponent(), BoneName=head, BaseName=spine_05))
   > Set NeckXCrouch(NeckXCrouch:double=MapRangeClamped(Value=BreakRotator(InRot=@n12.Rotation).Pitch, InRangeA=30, InRangeB=40)) -> (Output_Get:double)
   n13 = GetOwnerSkeletalMeshComponent()
   n14 = BreakRotator(InRot=BreakTransform(InTransform=GetDeltaTransformFromRefPose(self=@n13, BoneName=head, BaseName=spine_03)).Rotation)
   > Set NeckYDown(NeckYDown:double=MapRangeClamped(Value=float + float(A=@n14.Pitch), InRangeA=-3, InRangeB=-5)) -> (Output_Get:double)
   > n15 = Branch(Condition=Get bIsCrouchSprinting())
   n16 = GetOwnerSkeletalMeshComponent()
   n17 = BreakRotator(InRot=BreakTransform(InTransform=GetDeltaTransformFromRefPose(self=@n16, BoneName=head, BaseName=spine_05)).Rotation)
   > Set Neck twist R(Neck twist R=MapRangeClamped(Value="Absolute (Float)"(A=@n17.Pitch), InRangeA=10, InRangeB=40))
   n18 = BreakTransform(InTransform=GetDeltaTransformFromRefPose(self=GetOwnerSkeletalMeshComponent(), BoneName=neck_01, BaseName=spine_05))
   > Set Extremewalkup(Extremewalkup:double=MapRangeClamped(Value=BreakRotator(InRot=@n18.Rotation).Pitch, InRangeA=10, InRangeB=22)) -> (Output_Get:double)
   Set Neck twist R(execute=@n15.else, Neck twist R=0)
   > Set Extremewalkup(Extremewalkup=0)
   Set Neck Alpha(execute=@n9.else, Neck Alpha=0)
   > Set Neck Rcrouch(Neck Rcrouch=0)
   > Set NeckXCrouch(NeckXCrouch=0)
   > Set NeckYDown(NeckYDown=0)
   n19 = Branch(execute=@n1.then_3, Condition=OR Boolean(A=Get bIsFalling(), B=Get bIsJumping()))
   n20 = BreakTransform(InTransform=GetDeltaTransformFromRefPose(self=GetOwnerSkeletalMeshComponent(), BoneName=thigh_r, BaseName=pelvis))
   > Set thigh_rot_r(thigh_rot_r=MapRangeClamped(Value=BreakRotator(InRot=@n20.Rotation).Yaw, InRangeA=-40, InRangeB=-120))
   Set thigh_rot_r(execute=@n19.else, thigh_rot_r=0)
   > n21 = Branch(Condition=Get bIsCrouching())
   n22 = BreakTransform(InTransform=GetDeltaTransformFromRefPose(self=GetOwnerSkeletalMeshComponent(), BoneName=thigh_r, BaseName=pelvis))
   > Set Thigh R Crouch(Thigh R Crouch:double=MapRangeClamped(Value=BreakRotator(InRot=@n22.Rotation).Yaw, InRangeA=-70, InRangeB=-120)) -> (Output_Get:double)
   Set Thigh R Crouch(execute=@n21.else, Thigh R Crouch=0)
   n23 = Event OnBeginCrouching()
   ResetDynamics(execute=[@n23.then, @n24.then], InTeleportType=ResetPhysics)
   > ResetDynamics(InTeleportType=TeleportPhysics)
   n24 = Event OnEndCrouching()
   Get bIsCrouchSprinting()
   Get bIsSprinting()
   Get bIsFalling()
  ]}
]
variables [
 ControlRigAlpha {type=number}
 CrouchingAnim {type=boolean}
 Extremewalkup {type=number}
 Gravity Override Params {
  type=GravityOverrideParamsStruct
  default {skydive_FloatingDown {gravityOverride=<0 -700 300>}, skydive_GoingUp {gravityOverride=<0 -300 600>}}
 }
 GravityOverride {type=Vector}
 Head Tilt Alpha {type=number}
 Neck Alpha {type=number}
 Neck L Alpha {type=number}
 Neck R Alpha {type=number}
 Neck Rcrouch {type=number}
 Neck twist R {type=number}
 NeckXCrouch {type=number}
 NeckYDown {type=number}
 Skel_Mesh {type=SkeletalMeshComponent}
 Thigh R Crouch {type=number}
 thigh_rot_l {type=number}
 thigh_rot_r {type=number}
]
